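# Niagara system for the holo swing trails of the FairyFlex pickaxe, built from three animation-trail ribbon emitters following skeletal-mesh socket pairs.
# EnergyTrail adds a second ribbon renderer, time-scrolled U coordinates and a cooling blue-violet color curve.
# Trail_Gold is a fading lavender ribbon despite its name, while Trail_Holo flashes white, with alpha and a material parameter driven by emitter age.
# All three scale brightness by time of day.
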NiagaraSystem /BRCosmetics/Weapons/FORT_Melee/Pickaxe_FairyFlex/FX/Holo/NS_Pickaxe_FairyFlex_Holo
dependencies {
 usedDataInterfaces [&NiagaraDataInterfaceColorCurve, &NiagaraDataInterfaceCurve, &NiagaraDataInterfaceSkeletalMesh]
 usedDynamicInputs []
 usedModules [
  &self:SystemSpawnScript.SystemScriptSource.SystemScriptGraph.NiagaraNodeAssignment_0.SetVariables_#1
  &SystemState
 ]
}
emitters {
 EnergyTrail {
  emitterUpdateScript [
   &EmitterState()
   &FN_SpawnAnimTrail(Spawn Interval=0)
  ]
  particleSpawnScript [
   &FN_InitializeAnimTrailBoneIndecies(First Bone or Socket Index=3, Second Bone or Socket Index=4)
   &FN_InitializeAnimTrail(Lifetime=0.22, SkeletalMesh=$System.Skeletal Mesh)
  ]
  particleUpdateScript [
   &FN_UpdateAnimTrail(SkeletalMesh=$System.Skeletal Mesh)
   &ScaleColor(
    Scale Alpha=&FloatFromCurve(
     FloatCurve=DataInterface() {propertyValues {Curve {keys [
         {arriveTangent=-1.6414, interpMode=RCIM_Cubic, leaveTangent=-1.6414, tangentMode=RCTM_User, time=0, value=1}
         {interpMode=RCIM_Cubic, tangentMode=RCTM_User, time=1, value=0}
        ]}}}
     Scale Curve=0.5
    )
    Scale RGB=&Multiply_Vector(
     A=&MakeVectorFromLinearColorRGB(Color=&ColorFromCurve(ColorCurve=DataInterface() {propertyValues {
         AlphaCurve {keys [<at 0 1 RCIM_Cubic>,<at 1 1 RCIM_Cubic>]}
         BlueCurve {keys [<at 0.0021 10>, {arriveTangent=-15.7241, interpMode=RCIM_Cubic, leaveTangent=-15.7241, time=0.234, value=2}, <at 0.5871 0.8>, <at 1 0.4787>]}
         GreenCurve {keys [
           <at 0.0021 7.3424>
           {arriveTangent=-12.163, interpMode=RCIM_Cubic, leaveTangent=-12.163, time=0.234, value=0.6847}
           <at 0.5871 0.226>
           <at 1 0.1316>
          ]}
         RedCurve {keys [
           <at 0.0021 3.0927>
           {arriveTangent=-4.2589, interpMode=RCIM_Cubic, leaveTangent=-4.2589, time=0.234, value=0.6977}
           <at 0.5871 0.6009>
           <at 1 0.5>
          ]}
        }}))
     B=&VectorFromFloat(Value=&FNTimeOfDayScalar(Night Scale=0.15, Morning Scale=0.7, Day Scale=0.5, Evening Scale=0.15))
    )
    ScaleA=true
    ScaleRGB=true
   )
   &ParticleState()
   /SetVariables(Particles.RibbonU0Override=&Multiply_Float(A=$Engine.Time, B=4)) {isSetParameters=true}
   &CameraOffset(Camera Offset Amount=10)
   &ScaleRibbonWidth(Ribbon Width Scale=&FloatFromCurve(FloatCurve=DataInterface() {propertyValues {Curve {keys [
         {arriveTangent=-2.083, interpMode=RCIM_Cubic, leaveTangent=-2.083, tangentMode=RCTM_User, time=0.25, value=1}
         {tangentMode=RCTM_User, time=1, value=0.25}
        ]}}}))
   /SetVariables(
    Particles.SecondaryColor:LinearColor=&MultiplyLinearColor(
     Linear Color=<rgba 0.0854 0.0769 0.1 0.8>
     Scale Factor=&MakeVector4(
      X=1
      Y=1
      Z=1
      W=&Multiply_Float(
       A=&MakeFloatFromLinearColor(LinearColor=$Particles.Color, Channel:ENiagaraLinearColor_Channels=::NewEnumerator3)
       B=&Multiply_Float(A=5, B=&FNTimeOfDayScalar(Night Scale=1, Morning Scale=1, Evening Scale=1.5))
      )
     )
    )
    Particles.SecondaryWidth:NiagaraFloat=&Multiply_Float(A=$Particles.RibbonWidth, B=1.1)
   ) {isSetParameters=true}
  ]
  rendererClasses [&NiagaraRibbonRendererProperties]
  renderers [
   {rendererClass=&NiagaraRibbonRendererProperties}
   {rendererClass=&NiagaraRibbonRendererProperties, rendererIndex=1}
  ]
 }
 Trail_Gold {
  emitterUpdateScript [
   &EmitterState()
   &FN_SpawnAnimTrail_v2()
  ]
  particleSpawnScript [
   /Game/Effects/Niagara/Modules/AnimTrail/FN_InitializeAnimTrail/FN_InitializeAnimTrailBoneIndecies(Second Bone or Socket Index=1)
   &FN_InitializeAnimTrail(Color=<rgba 0.6574 0.4143 1 1>, Lifetime=0.22, SkeletalMesh=$System.Skeletal Mesh) {module=FN_InitializeAnimTrail001}
  ]
  particleUpdateScript [
   &FN_UpdateAnimTrail(SkeletalMesh=$System.Skeletal Mesh)
   &ScaleRibbonWidth(Ribbon Width Scale=&FloatFromCurve(FloatCurve=DataInterface() {propertyValues {Curve {keys [
         {arriveTangent=-1.2579, interpMode=RCIM_Cubic, time=0, value=1}
         {arriveTangent=-0.9664, interpMode=RCIM_Cubic, leaveTangent=-0.9664, tangentMode=RCTM_User, time=1, value=0}
        ]}}}))
   &ScaleColor(
    Scale Alpha=&Multiply_Float(
     A=&FloatFromCurve(FloatCurve=DataInterface() {propertyValues {Curve {keys [
          {interpMode=RCIM_Cubic, leaveTangent=-2.5656, tangentMode=RCTM_Break, time=0, value=1}
          {arriveTangent=-0.151, tangentMode=RCTM_User, time=0.5, value=0}
         ]}}})
     B=&FNTimeOfDayScalar(Night Scale=0.5, Morning Scale=0.75, Evening Scale=0.75)
    )
    Scale RGB=&VectorFromFloat(Value=&FNTimeOfDayScalar(Night Scale=0.5, Morning Scale=0.75, Evening Scale=0.75))
    ScaleA=true
    ScaleRGB=true
   ) {module=ScaleColor002}
  ]
  rendererClasses [&NiagaraRibbonRendererProperties]
  renderers [{rendererClass=&NiagaraRibbonRendererProperties}]
 }
 Trail_Holo {
  emitterUpdateScript [
   &EmitterState()
   &FN_SpawnAnimTrail_v2()
  ]
  particleSpawnScript [
   &FN_InitializeAnimTrailBoneIndecies(Second Bone or Socket Index=1)
   &FN_InitializeAnimTrail(Color=<rgba 2 2 2 0.2>, Lifetime=0.15, SkeletalMesh=$System.Skeletal Mesh) {module=FN_InitializeAnimTrail001}
  ]
  particleUpdateScript [
   &FN_UpdateAnimTrail(SkeletalMesh=$System.Skeletal Mesh)
   &ScaleRibbonWidth(Ribbon Width Scale=&FloatFromCurve(FloatCurve=DataInterface() {propertyValues {Curve {keys [
         {arriveTangent=-1.2579, interpMode=RCIM_Cubic, time=0, value=1}
         {arriveTangent=-0.9664, interpMode=RCIM_Cubic, leaveTangent=-0.9664, tangentMode=RCTM_User, time=1, value=0}
        ]}}})) {module=ScaleRibbonWidth001}
   &ScaleColor(
    Scale Alpha=&Multiply_Float(
     A=&FloatFromCurve(
      FloatCurve=DataInterface() {propertyValues {Curve {keys [
          {arriveTangent=-3.9039, interpMode=RCIM_Cubic, leaveTangent=-3.9039, tangentMode=RCTM_User, time=0.5, value=1}
          {arriveTangent=-2.0179, interpMode=RCIM_Cubic, leaveTangent=-2.0179, tangentMode=RCTM_User, time=1, value=0}
         ]}}}
      Scale Curve=&FloatFromCurve(
       FloatCurve=DataInterface() {propertyValues {Curve {keys [
           {interpMode=RCIM_Cubic, tangentMode=RCTM_User, time=0.05, value=0}
           {interpMode=RCIM_Cubic, tangentMode=RCTM_User, time=0.08, value=1}
           {interpMode=RCIM_Cubic, leaveTangent=-8.9897, time=0.35, value=0}
          ]}}}
       CurveIndex=$Emitter.Age
      )
     )
     B=&FNTimeOfDayScalar(Night Scale=0.5, Morning Scale=0.75, Evening Scale=0.75)
    )
    Scale RGB=&VectorFromFloat(Value=&FloatFromCurve(
      FloatCurve=DataInterface() {propertyValues {Curve {keys [
          {interpMode=RCIM_Cubic, tangentMode=RCTM_User, time=0, value=1}
          {arriveTangent=-1.5604, interpMode=RCIM_Cubic, leaveTangent=-1.5604, tangentMode=RCTM_User, time=1, value=0}
         ]}}}
      Scale Curve=&FNTimeOfDayScalar(Night Scale=0.8, Morning Scale=0.9, Evening Scale=0.9)
     ))
    ScaleA=true
    ScaleRGB=true
   ) {module=ScaleColor002}
   &DynamicMaterialParameters(
    Index 1 Param 1=&FloatFromCurve(
     FloatCurve=DataInterface() {propertyValues {Curve {keys [{arriveTangent=-0.4165, interpMode=RCIM_Cubic, time=0.1, value=-0.875}, {arriveTangent=12.7593, leaveTangent=12.7593, time=0.3, value=0.5}]}}}
     CurveIndex=$Emitter.Age
    )
    Module.Index1Param0Write=true
    Module.Index1Param1Write=true
    Module.Index1Param2Write=true
    Module.Index1Param3Write=true
    Write Parameter Index 0=false
    Write Parameter Index 1=true
   ) {module=DynamicMaterialParameters001}
  ]
  rendererClasses [&NiagaraRibbonRendererProperties]
  renderers [{rendererClass=&NiagaraRibbonRendererProperties}]
 }
}
summary {
 systemName=NS_Pickaxe_FairyFlex_Holo
 emitters [
  {emitterName=EnergyTrail, rendererClasses [&NiagaraRibbonRendererProperties]}
  {emitterName=Trail_Gold, rendererClasses [&NiagaraRibbonRendererProperties]}
  {emitterName=Trail_Holo, rendererClasses [&NiagaraRibbonRendererProperties]}
 ]
 userVariables [
  User.RibbonIDOffset {type=NiagaraInt32}
  User.SwingSectionIndex {type=NiagaraInt32}
 ]
}
